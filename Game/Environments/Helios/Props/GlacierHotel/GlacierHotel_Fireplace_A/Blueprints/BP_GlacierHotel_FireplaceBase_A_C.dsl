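; Blueprint for the Glacier Hotel fireplace base prop, which reacts to dynamic destruction.
; When dynamic destruction starts, it calls the parent handler and then destroys its NS_Helios_Fireplace_Big component, removing the fireplace effect.
; The construction script only calls the parent, and BeginPlay, overlap and Tick events are empty.
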(event EventBeginPlay)

(event Collision|EventActorBeginOverlap
  (params (OtherActor :type Actor%20Object%20Reference)))

(event EventTick
  (params (DeltaSeconds :type Float%20%28single-precision%29)))

(event DynamicDestruction|EventOnDynamicDestructionStarted
  (|CallParentFunction)
  (Components|DestroyComponent (Variables|Default|GetNS_Helios_Fireplace_Big)))

(fn ConstructionScript
  (|CallParentFunction))
